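# Control Rig for the Sweet Teriyaki Blue backpack, a fish-like rig with jaw, eyes, fins and tail, that adds squash and stretch to its body.
# It reads the absolute local Z offset of the dyn_jiggle bone and remaps it into two weights around a shared midpoint.
# Two chained ModifyTransforms use them to scale dyn_body, slightly taller and narrower for larger offsets, flatter and longer for smaller ones.
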ControlRigBlueprint /BRCosmetics/Accessories/FORT_Backpacks/Backpack_M_MED_Sweet_Teriyaki_Blue/Meshes/M_MED_Sweet_Teriyaki_Blue_Pack_CtrlRig
graphs [
 RigVMFunctionLibrary {links [], nodes []}
 RigVMModel {nodes [
   RigUnit_BeginExecution = &RigUnit_BeginExecution()
   n1 = /Script/RigVM.RigVMFunction_MathFloatAbs(Value=&RigUnit_GetTransform(Item=(Type=Bone,Name="dyn_jiggle"), Space=LocalSpace, bInitial=false).Transform.Translation.Z) {is=struct RigVMUnitNode, template="Absolute::Execute(in Value,out Result)"}
   &RigUnit_ModifyTransforms(
    ItemToModify=((Item=(Type=Bone),Transform=(Rotation=(X=0,Y=0,Z=0,W=1),Translation=(X=0,Y=0,Z=0),Scale3D=(X=1,Y=1,Z=1))),(Item=(Type=Bone,Name="dyn_body"),Transform=(Rotation=(X=0,Y=0,Z=0,W=1),Translation=(X=0,Y=0,Z=0),Scale3D=(X=0.975,Y=1,Z=1.025))))
    WeightMinimum=0
    Weight=&RigVMFunction_MathFloatRemap(SourceMaximum=8, SourceMinimum=4, Value=@n1.Result).Result
    ExecutePin=@RigUnit_BeginExecution.ExecutePin
   )
   > &RigUnit_ModifyTransforms(
    ItemToModify=((Item=(Type=Bone),Transform=(Rotation=(X=0,Y=0,Z=0,W=1),Translation=(X=0,Y=0,Z=0),Scale3D=(X=1,Y=1,Z=1))),(Item=(Type=Bone,Name="dyn_body"),Transform=(Rotation=(X=0,Y=0,Z=0,W=1),Translation=(X=0,Y=0,Z=0),Scale3D=(X=1.1,Y=1,Z=0.9))))
    WeightMinimum=0
    Weight=&RigVMFunction_MathFloatRemap(SourceMaximum=2, SourceMinimum=4, Value=@n1.Result).Result
   )
  ]}
]
hierarchy [
 root {loc=<0 0 0>, rot=<0 0 0 1>}
 pack {parent=root, loc=<0 -0.7316 0>, rot=<0 0 0 1>}
 dyn_body {parent=pack, loc=<0 -6.4701 0>, rot=<0 0 0 1>}
 dyn_eye_r {parent=dyn_body, loc=<-5.4388 -4.7174 1.9364>, rot=<0.5131 -0.8426 -0.0851 0.1398>}
 dyn_eye_l {parent=dyn_body, loc=<-5.4388 4.7174 1.9364>, rot=<-0.5316 -0.8311 -0.163 -0.0151>}
 dyn_jaw {parent=dyn_body, loc=<-10.7071 -0.0003 -1.4741>, rot=<0 -0.9791 0 0.2032>}
 dyn_fin_bottom1 {parent=dyn_body, loc=<-2.1348 -0.0003 -11.2381>, rot=<0 -0.7431 0 -0.6692>}
 dyn_fin_bottom2 {parent=dyn_fin_bottom1, loc=<3.7915 0 0>, rot=<0 0 0 1>}
 dyn_fin_bottom3 {parent=dyn_body, loc=<6.9614 -0.0003 -9.6268>, rot=<0 -0.441 0 -0.8975>}
 dyn_tail1 {parent=dyn_body, loc=<11.7212 -0.0003 -1.4282>, rot=<0 -0.0009 0 -1>}
 dyn_tail2 {parent=dyn_tail1, loc=<2.114 0 3.3009>, rot=<0 0.4806 0 -0.877>}
 dyn_tail3 {parent=dyn_tail1, loc=<1.3116 0 -4.2414>, rot=<0 -0.5687 0 -0.8226>}
 dyn_fin_top3 {parent=dyn_body, loc=<8.5728 -0.0003 7.1621>, rot=<0 0.3133 0 -0.9497>}
 dyn_fin_top1 {parent=dyn_body, loc=<2.0526 0 9.8404>, rot=<0 0.5276 0 -0.8495>}
 dyn_fin_top2 {parent=dyn_fin_top1, loc=<5.115 0 0>, rot=<0 0 0 1>}
 dyn_jiggle {parent=dyn_body, loc=<0 0 0>, rot=<0 0 0 1>}
]
