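# Physics asset for the Sweet Teriyaki Blue backpack cosmetic, giving a fish-like creature's fins, tail, eyes and jaw physics bodies for secondary motion.
# Small capsules cover the parts, with a sphere on a jiggle bone.
# Constraints radiate from dyn_body, with short fin chains and a tail that forks after its first bone.
# Joints set per-part swing and narrow twist angles, except the jiggle joint, whose angular motion is locked.
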PhysicsAsset /BRCosmetics/Accessories/FORT_Backpacks/Backpack_M_MED_Sweet_Teriyaki_Blue/Meshes/M_MED_Sweet_Teriyaki_Blue_Pack_Physics
bodies {
 dyn_body [{shapeType=Capsule, length=1, radius=3}]
 dyn_eye_l [{shapeType=Capsule, length=3, radius=4, shapeName=dyn_eye_l_capsule, center=<1 0 0>, rotation=<rot 90 0 0>}]
 dyn_eye_r [{shapeType=Capsule, length=3, radius=4, shapeName=dyn_eye_l_capsule, center=<1 0 0>, rotation=<rot 90 0 0>}]
 dyn_fin_bottom1 [{shapeType=Capsule, length=1.75, radius=2, shapeName=dyn_fin_top1_capsule, rotation=<rot 90 0 0>}]
 dyn_fin_bottom2 [{shapeType=Capsule, length=1.75, radius=2, shapeName=dyn_fin_top1_capsule, center=<1 0 0>, rotation=<rot 90 0 0>}]
 dyn_fin_bottom3 [{shapeType=Capsule, length=1.75, radius=2, shapeName=dyn_fin_top1_capsule, center=<2 0 0>, rotation=<rot 90 0 0>}]
 dyn_fin_top1 [{shapeType=Capsule, length=3, radius=2, shapeName=dyn_fin_top1_capsule, rotation=<rot 90 0 0>}]
 dyn_fin_top2 [{shapeType=Capsule, length=1.75, radius=2, shapeName=dyn_fin_top1_capsule, center=<2 0 0>, rotation=<rot 90 0 0>}]
 dyn_fin_top3 [{shapeType=Capsule, length=1.75, radius=2, shapeName=dyn_fin_top1_capsule, center=<2 0 0>, rotation=<rot 90 0 0>}]
 dyn_jaw [{shapeType=Capsule, length=1, radius=2, shapeName=dyn_eye_l_capsule, center=<0.5 0 0>, rotation=<rot 90 0 0>}]
 dyn_jiggle [{shapeType=Sphere, radius=3, shapeName=dyn_jiggle_sphere}]
 dyn_tail1 [{shapeType=Capsule, length=6, radius=2, shapeName=dyn_fin_top1_capsule}]
 dyn_tail2 [{shapeType=Capsule, length=1.75, radius=2, shapeName=dyn_fin_top1_capsule, center=<2 0 0>, rotation=<rot 90 0 0>}]
 dyn_tail3 [{shapeType=Capsule, length=1.75, radius=2, shapeName=dyn_fin_top1_capsule, center=<2 0 0>, rotation=<rot 90 0 0>}]
}
constraints [
 {bone1Name=dyn_fin_top1, bone2Name=dyn_body, swing1LimitDegrees=25, swing2LimitDegrees=50, twistLimitDegrees=1}
 {bone1Name=dyn_fin_top2, bone2Name=dyn_fin_top1, swing1LimitDegrees=45, swing2LimitDegrees=50, twistLimitDegrees=1}
 {bone1Name=dyn_fin_top3, bone2Name=dyn_body, swing1LimitDegrees=25, swing2LimitDegrees=10, twistLimitDegrees=1}
 {bone1Name=dyn_fin_bottom1, bone2Name=dyn_body, swing1LimitDegrees=25, swing2LimitDegrees=5, twistLimitDegrees=1}
 {bone1Name=dyn_fin_bottom3, bone2Name=dyn_body, swing1LimitDegrees=25, swing2LimitDegrees=10, twistLimitDegrees=1}
 {bone1Name=dyn_fin_bottom2, bone2Name=dyn_fin_bottom1, swing1LimitDegrees=25, swing2LimitDegrees=5, twistLimitDegrees=1}
 {bone1Name=dyn_tail1, bone2Name=dyn_body, swing1LimitDegrees=35, swing2LimitDegrees=5, twistLimitDegrees=1}
 {bone1Name=dyn_tail2, bone2Name=dyn_tail1, swing1LimitDegrees=35, swing2LimitDegrees=5, twistLimitDegrees=1}
 {bone1Name=dyn_tail3, bone2Name=dyn_tail1, swing1LimitDegrees=35, swing2LimitDegrees=5, twistLimitDegrees=1}
 {bone1Name=dyn_eye_l, bone2Name=dyn_body, swing1LimitDegrees=45, swing2LimitDegrees=45, twistLimitDegrees=1}
 {bone1Name=dyn_eye_r, bone2Name=dyn_body, swing1LimitDegrees=45, swing2LimitDegrees=45, twistLimitDegrees=1}
 {bone1Name=dyn_jaw, bone2Name=dyn_body, swing1LimitDegrees=5, swing2LimitDegrees=15, twistLimitDegrees=1}
 {
  bone1Name=dyn_jiggle
  bone2Name=dyn_body
  swing1LimitDegrees=45
  swing1Motion=Locked
  swing2LimitDegrees=45
  swing2Motion=Locked
  twistLimitDegrees=45
  twistMotion=Locked
 }
]
